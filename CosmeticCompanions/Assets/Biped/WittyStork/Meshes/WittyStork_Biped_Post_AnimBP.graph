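# Post-process AnimBlueprint for the WittyStork biped companion mesh, driving pose-asset deformation, apparently facial, from remapped animation curves.
# Remap Curves From Mesh applies the blink remap expressions, with the owning component cached at init as source and attached-parent lookup enabled.
# The remapped curves feed the Pose Asset blend node.
# A BrowMask layered blend with an unconnected blend pose lets DisableBrowWeight fade out the brow posing.
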AnimBlueprint /CosmeticCompanions/Assets/Biped/WittyStork/Meshes/WittyStork_Biped_Post_AnimBP
target_skeleton=&/WittyStork_Biped_Skeleton
graphs [
 AnimGraph {nodes [
   n1 = Input Pose()
   n2 = Get mesh component() -> (mesh component:Skeletal Mesh Component Object Reference)
   n3 = Layered blend per bone(
    BasePose=Pose!(
     PoseAsset:Pose Asset Object Reference=Get PoseAsset() -> (PoseAsset:Pose Asset Object Reference)
     SourcePose:Pose Link Structure=Remap Curves From Mesh!(
      SourcePose:Pose Link Structure=@n1
      CurveExpressionsDataAsset:Curve Expressions Data Asset Object Reference=&/WittyStork_Blink_Remap
      SourceMeshComponent:Skeletal Mesh Component Weak object=@n2
     ) -> (Pose:Pose Link Structure) {is=title AnimGraphNode_RemapCurvesFromMesh, bUseAttachedParent=true, expressionSource=DataAsset}
    ) -> (Pose:Pose Link Structure) {is=title AnimGraphNode_PoseBlendNode, blendOption=Linear}
    BlendWeights_0=Get DisableBrowWeight() -> (DisableBrowWeight:double)
   ) {blendMode=BlendMask, blendMasks [&WittyStork_Biped_Skeleton:BrowMask], layerSetup []}
   Output Pose(Result=@n3)
  ]}
 EventGraph {nodes [
   Event BlueprintInitializeAnimation()
   > Set mesh component(mesh component:Skeletal Mesh Component Object Reference=GetOwningComponent()) -> (Output_Get:Skeletal Mesh Component Object Reference)
  ]}
]
variables [
 DisableBrowWeight {type=number}
 mesh component {type=SkeletalMeshComponent}
 PoseAsset {type=PoseAsset, default=&/WittyStork_Biped_PoseAsset}
]
